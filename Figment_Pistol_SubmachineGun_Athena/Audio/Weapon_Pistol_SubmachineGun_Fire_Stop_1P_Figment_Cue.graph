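# First-person fire-stop sound cue for the Figment submachine pistol, named as the sound that ends a burst of gunfire.
# A single wave player feeds a borrowed Military SMG close fire-stop sample through a modulator left at default settings.
# Playback uses first-person weapon attenuation, the ranged-fire sound class and gunfire concurrency, with slightly reduced volume and the debug flag left on.
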SoundCue /Figment_Pistol_SubmachineGun_Athena/Audio/Weapon_Pistol_SubmachineGun_Fire_Stop_1P_Figment_Cue
root=Modulator_0
nodes [
 {id=Modulator_0, children [WavePlayer_0]}
 {id=WavePlayer_0, properties {soundWaveAssetPtr=&/MilitarySMG_Close_FireStop_03}}
]
properties {
 attenuationSettings=&Fort_Weapon_Projectile_FirstPerson_Athena_Att
 bDebug=true
 maxDistance=2097152
 soundClassObject=&Fort_WeaponRanged_Fire_1P_PSM
 volumeMultiplier=0.868
 concurrencySet [&Weapon_Gunfire_FirstPerson_Conc]
}
